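# Physics asset for the SK_Rocket_Ride_Export skeletal mesh, a rideable-rocket firework prop under BR cosmetics animation.
# A single small capsule body on the root_anim_offset bone, rolled about 90 degrees so its long axis lies along the bone's Y axis, provides the prop's collision shape.
# It has no constraints, so it defines no jointed or ragdoll chain.
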PhysicsAsset /BRCosmetics/Animation/Props/Fireworks/Meshes/SK_Rocket_Ride_Export_PhysicsAsset
bodies {root_anim_offset [{shapeType=Capsule, length=3.9135, radius=2.2062, center=<0.0074 0.5037 -0.0492>, rotation=<rot -0.006 179.8851 -89.4305>}]}
constraints []
